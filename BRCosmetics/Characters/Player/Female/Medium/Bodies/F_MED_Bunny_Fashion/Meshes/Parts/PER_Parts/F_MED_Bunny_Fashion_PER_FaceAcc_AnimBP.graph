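# Post-attach AnimBP for the Bunny Fashion female face accessory part, adding physics to hair and hood bones.
# Each of two branches copies the attached parent's pose and runs RigidBody simulation in spine_04 space with its own physics asset.
# On customization load, variant tags set hair_base_active: true selects the hood physics branch with additive back-hair and bun offsets.
# False selects the other physics branch, without offsets.
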AnimBlueprint /BRCosmetics/Characters/Player/Female/Medium/Bodies/F_MED_Bunny_Fashion/Meshes/Parts/PER_Parts/F_MED_Bunny_Fashion_PER_FaceAcc_AnimBP
target_skeleton=&/F_MED_Bunny_Fashion_PER_FaceAcc_Skeleton
graphs [
 AnimGraph {nodes [
   n1 = RigidBody(ComponentPose=Local To Component(LocalPose=Copy Pose From Mesh() {bUseAttachedParent=true})) {
    bClampLinearTranslationLimitToRefPose=true
    bOverrideWorldGravity=true
    overridePhysicsAsset=&/F_MED_Bunny_Fashion_E_Physics
    simulationSpace=BaseBoneSpace
    alphaBoolBlend {blendOption=Linear}
    baseBoneRef {boneName=spine_04}
    componentLinearAccScale=<0.375 0.375 0.375>
    componentLinearVelScale=<0.375 0.375 0.375>
    overrideWorldGravity {z=-980}
   }
   n2 = RigidBody(ComponentPose=Local To Component(LocalPose=Copy Pose From Mesh() {bUseAttachedParent=true})) {
    bClampLinearTranslationLimitToRefPose=true
    bOverrideWorldGravity=true
    overridePhysicsAsset=&/F_MED_Bunny_Fashion_E_Hood_Physics
    simulationSpace=BaseBoneSpace
    alphaBoolBlend {blendOption=Linear}
    baseBoneRef {boneName=spine_04}
    componentAppliedLinearAccClamp=<5000 5000 5000>
    componentLinearAccScale=<0.375 0.375 0.375>
    componentLinearVelScale=<0.375 0.375 0.375>
    overrideWorldGravity {z=-980}
   }
   n3 = "Transform (Modify) Bone - Bone: dyn_hood_up_back_hair"(ComponentPose=@n2, Alpha=1, Translation="0,3,0", Rotation="0,0,0", Scale="1,1,1") {translationMode=BMM_Additive, alphaBoolBlend {blendOption=Linear}, boneToModify {boneName=dyn_hood_up_back_hair}, translation {y=3}}
   n4 = "Transform (Modify) Bone - Bone: per2_hair_bun"(ComponentPose=@n3, Alpha=1, Translation="0,1,-1", Rotation="0,0,0", Scale="1,1,1") {translationMode=BMM_Additive, alphaBoolBlend {blendOption=Linear}, boneToModify {boneName=per2_hair_bun}, translation {y=1, z=-1}}
   Output Pose(Result=Blend Poses by bool(
     BlendPose_0=Component To Local(ComponentPose=@n4)
     BlendPose_1=Component To Local(ComponentPose=@n1)
     bActiveValue=Get hair_base_active() -> (hair_base_active:bool)
    ) {blendType=Linear}) {layerGroup=None}
  ]}
 EventGraph {nodes [
   Event OnCharacterCustomizationLoaded()
   > n1 = Branch(Condition=CharacterItemDefinitionHasAnyMetaTag(Tags=Make Array!("[0]"=(TagName="Cosmetics.Variant.Property.Stage1"), "[1]":Gameplay Tag Structure=(TagName="Cosmetics.Variant.Property")) -> (Array)))
   > Set hair_base_active(hair_base_active:bool=true) -> (Output_Get:bool)
   Set hair_base_active(execute=@n1.else, hair_base_active=false)
   Event BlueprintInitializeAnimation()
   > Set Mesh_To_Copy(Mesh_To_Copy:Skeletal Mesh Component Object Reference=GetOwnerSkeletalMeshComponent()) -> (Output_Get:Skeletal Mesh Component Object Reference)
   TryGetPawnOwner()
   Event BlueprintUpdateAnimation()
  ]}
]
variables [
 hair_base_active {type=boolean}
 Mesh_To_Copy {type=SkeletalMeshComponent}
]
